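# RFL_DebugTools is an editor utility function library for tracing which assets changed in a time span, and their dependencies.
# Change queries pass packages, dates and client name to Python DebugToolLib, reloaded each call, returning packages, dates and changelists, optionally by user.
# FilterAssetsByChangeDate simply wraps DidAssetsChangeInTimeSpan.
# FindDependencies recursively walks Asset Registry dependencies, including soft references, using a checked set to skip visited packages.
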EditorUtilityBlueprint /Game/NeverCook/EditorUtilities/DebugTools/FunctionLibs/RFL_DebugTools
graphs [
 DidAssetsChangeInTimeSpan {nodes [
   n1 = DidAssetsChangeInTimeSpan!() -> (then, AssetPackageNamesIn:Array of Names, DateEarliest:Date Time Structure, DateLatest:Date Time Structure, ClientName:String) {is=title K2Node_FunctionEntry}
   > n2 = Execute Python Script!(
    execute
    PythonScript="import TechArt.Tools.DebugTools.DebugToolLib as tool_lib\r\nfrom importlib import reload\r\nreload(tool_lib)\r\npackages_out, dates_out, changelists_out  = tool_lib.filterAssetsByChangeDate(in_assets, in_client, in_date_early, in_date_latest)"
    in_assets:Array of Names=@n1
    in_client:String=@n1.ClientName
    in_date_early:Date Time Structure=@n1.DateEarliest
    in_date_latest:Date Time Structure=@n1.DateLatest
   ) -> (then, ReturnValue, packages_out:Array of Names, dates_out:Array of Integers, changelists_out:Array of Integers) {inputs [in_assets, in_client, in_date_early, in_date_latest], outputs [packages_out, dates_out, changelists_out]}
   > Return Node!(
    execute
    ChangedPackageNames:Array of Names=@n2.packages_out
    UTCDates:Array of Integers=@n2.dates_out
    Changelists:Array of Integers=@n2.changelists_out
   )
  ]}
 FilterAssetsByChangeDate {nodes [
   n1 = FilterAssetsByChangeDate!() -> (then, AssetPackageNamesIn:Array of Names, DateEarliest:Date Time Structure, DateLatest:Date Time Structure, ClientName:String) {is=title K2Node_FunctionEntry}
   > n2 = DidAssetsChangeInTimeSpan!(
    execute
    AssetPackageNamesIn:Array of Names=@n1
    DateEarliest:Date Time Structure=@n1.DateEarliest
    DateLatest:Date Time Structure=@n1.DateLatest
    ClientName:String=@n1.ClientName
    __WorldContext:Object Reference=Get l_world()
   ) -> (then, ChangedPackageNames:Array of Names, UTCDates:Array of Integers, Changelists:Array of Integers) {is=title K2Node_CallFunction}
   > Return Node(ChangedPackageNames=@n2, UTCDates=@n2.UTCDates, Changelists=@n2.Changelists)
  ]}
 FindDependencies {nodes [
   n1 = FindDependencies!() -> (then, PackageName:Name, CheckedAssetsIn:Set of Names) {is=title K2Node_FunctionEntry}
   > Set l_checkedAssets(l_checkedAssets:Set of Names=@n1.CheckedAssetsIn) -> (Output_Get:Set of Names)
   > n2 = Get Dependencies!(
    execute
    self:Asset Registry Interface=GetAssetRegistry()
    PackageName:Name=@n1
    DependencyOptions:"Asset Registry Dependency Options Structure (by ref)"=Make Asset Registry Dependency Options(bIncludeSoftPackageReferences=True)
   ) -> (then, OutDependencies:Array of Names, ReturnValue:bool) {is=title K2Node_CallFunction}
   > Branch(Condition=Is Valid Index!(TargetArray:Array of Names=@n2, IndexToTest:int) -> (ReturnValue:bool))
   > n3 = GetAssets!(
    execute
    self:Asset Registry Interface=GetAssetRegistry()
    Filter:"ARFilter Structure (by ref)"=MakeARFilter!(
     PackageNames:Array of Names=@n2
     PackagePaths:Array of Names
     SoftObjectPaths:Array of Soft Object Path Structures
     ClassPaths:Array of Top Level Asset Path Structures
     RecursiveClassPathsExclusionSet:Set of Top Level Asset Path Structures
     ClassNames:Array of Names
     RecursiveClassesExclusionSet:Set of Names
     bRecursivePaths:bool=false
     bRecursiveClasses:bool=false
     bIncludeOnlyOnDiskAssets:bool=false
    ) -> (ReturnValue:ARFilter Structure)
    bSkipARFilteredAssets:bool=true
   ) -> (then, OutAssetData:Array of Asset Data Structures, ReturnValue:bool)
   > n4 = For Each Loop(Array:Array of Asset Data Structures=@n3) -> (Array Element:Asset Data Structure)
   n5 = To Array!(execute=@n4.Completed, A:Set of Names=Get l_checkedAssets() -> (l_checkedAssets:Set of Names)) -> (then, Result:Array of Names)
   > Return Node!(execute, CheckedAssetsOut:Array of Names=@n5)
   Set l_package(execute=@n4.LoopBody, l_package:Name=Break Asset Data(AssetData=@n4)) -> (Output_Get:Name=None)
   > n6 = Branch(Condition=Contains Item!(TargetSet:Set of Names=Get l_checkedAssets(), ItemToFind:"Name (by ref)"=Get l_package() -> (l_package:Name=None)) -> (ReturnValue:bool) {is=title K2Node_CallFunction})
   n7 = Sequence(execute=@n6.else)
   > Add!(execute, TargetSet:Set of Names=Get l_checkedAssets(), NewItem:"Name (by ref)"=Get l_package() -> (l_package=None)) -> (then) {is=title K2Node_CallFunction}
   n8 = FindDependencies!(
    execute=@n7.then_1
    PackageName:Name=Get l_package() -> (l_package=None)
    CheckedAssetsIn:Set of Names=Get l_checkedAssets()
    __WorldContext:Object Reference=Get l_world()
   ) -> (then, CheckedAssetsOut:Array of Names) {is=title K2Node_CallFunction}
   > Add Items!(execute, TargetSet:Set of Names=Get l_checkedAssets(), NewItems:Array of Names=@n8) -> (then)
  ]}
 GetUserChangedAssetsInTimeSpan {nodes [
   n1 = GetUserChangedAssetsInTimeSpan!() -> (
    then
    user_name:String
    DateEarliest:Date Time Structure
    DateLatest:Date Time Structure
    ClientName:String
    AssetPackageNamesIn:Array of Names
   ) {is=title K2Node_FunctionEntry}
   > n2 = Execute Python Script!(
    execute
    PythonScript="import TechArt.Tools.DebugTools.DebugToolLib as tool_lib\r\nfrom importlib import reload\r\nreload(tool_lib)\r\npackages_out, dates_out, changelists_out  = tool_lib.getChangedFilesByUserInTimespan(user_name, in_client, in_date_early, in_date_latest, in_assets)"
    in_assets=@n1.AssetPackageNamesIn
    in_client=@n1.ClientName
    in_date_early=@n1.DateEarliest
    in_date_latest=@n1.DateLatest
    user_name:String=@n1
   ) -> (then, ReturnValue, packages_out, dates_out, changelists_out) {inputs [in_assets, in_client, in_date_early, in_date_latest, user_name], outputs [packages_out, dates_out, changelists_out]}
   > Return Node!(
    execute
    ChangedPackageNames:Array of Names=@n2.packages_out
    UTCDates:Array of Integers=@n2.dates_out
    Changelists:Array of Integers=@n2.changelists_out
   )
  ]}
]
variables []
